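# Physics asset for the Tomcat backpack cosmetic mesh, giving its dangling straps, guns and zipper bodies and joints for secondary motion.
# A box body on the pack anchors three four-link strap chains (a, b, d), four two-part gun chains and a two-part zipper, all built from capsules.
# Straps, gun bases and zipper joints have tight angular limits, while each gun end rotates freely on its base.
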PhysicsAsset /BRCosmetics/Accessories/FORT_Backpacks/Backpack_M_MED_Tomcat/Meshes/M_MED_Tomcat_Pack_Physics
bodies {
 gun1_base [{shapeType=Capsule, length=4.5, radius=2, shapeName=gun1_base_capsule, center=<3.9502 0 0>, rotation=<rot 90 0 0>}]
 gun1_end [{shapeType=Capsule, length=1, radius=1, shapeName=gun1_end_capsule, center=<-0.8197 0 1.2674>, rotation=<rot 90 0 0>}]
 gun2_base [{shapeType=Capsule, length=4.5, radius=2, shapeName=gun2_base_capsule, center=<3.9502 0 0>, rotation=<rot 90 0 0>}]
 gun2_end [{shapeType=Capsule, length=1, radius=1, shapeName=gun2_end_capsule, center=<-1.2092 0 1.0999>, rotation=<rot 90 0 0>}]
 gun3_base [{shapeType=Capsule, length=3.5, radius=4.25, shapeName=gun3_base_capsule, center=<5.1515 0.0001 0>, rotation=<rot 90 0 0>}]
 gun3_end [{shapeType=Capsule, length=1, radius=1, shapeName=gun3_end_capsule, center=<-1.2388 0 0.7263>, rotation=<rot 90 0 0>}]
 gun4_base [{shapeType=Capsule, length=3.5, radius=4.25, shapeName=gun4_base_capsule, center=<5.1515 0.0001 0>, rotation=<rot 90 0 0>}]
 gun4_end [{shapeType=Capsule, length=1, radius=1, shapeName=gun4_end_capsule, center=<-1.3041 0 0.7494>, rotation=<rot 90 0 0>}]
 pack [{shapeType=Box, extentX=10.5, extentY=6.5, extentZ=31, shapeName=pack_box, center=<1.0573 -8.3201 -7.3726>, rotation=<rot 19.9096 0 0>}]
 strap1_a [{shapeType=Capsule, length=2, radius=1.75, shapeName=strap1_a_capsule, center=<2 0 0>, rotation=<rot 90 0 0>}]
 strap1_b [{shapeType=Capsule, length=2, radius=1.5, shapeName=strap1_b_capsule, center=<2.0164 0 0>, rotation=<rot 90 0 0>}]
 strap1_d [{shapeType=Capsule, length=1, radius=1, shapeName=strap1_d_capsule, center=<1.2686 0 0>, rotation=<rot 90 0 0>}]
 strap2_a [{shapeType=Capsule, length=2, radius=1.75, shapeName=strap2_a_capsule, center=<3.0701 0.0001 0>, rotation=<rot 90 0 0>}]
 strap2_b [{shapeType=Capsule, length=2, radius=1.5, shapeName=strap2_b_capsule, center=<2.4467 0 0>, rotation=<rot 90 0 0>}]
 strap2_d [{shapeType=Capsule, length=1, radius=1, shapeName=strap2_d_capsule, center=<1.2641 0 0>, rotation=<rot 90 0 0>}]
 strap3_a [{shapeType=Capsule, length=2, radius=1.75, shapeName=strap3_a_capsule, center=<3.4667 0 0>, rotation=<rot 90 0 0>}]
 strap3_b [{shapeType=Capsule, length=2, radius=1.5, shapeName=strap3_b_capsule, center=<2.6401 0 0>, rotation=<rot 90 0 0>}]
 strap3_d [{shapeType=Capsule, length=1, radius=1, shapeName=strap3_d_capsule, center=<1.7133 0 0>, rotation=<rot 90 0 0>}]
 strap4_a [{shapeType=Capsule, length=2, radius=1.75, shapeName=strap4_a_capsule, center=<2 -0.9697 -1.6207>, rotation=<rot 90 0 0>}]
 strap4_b [{shapeType=Capsule, length=2, radius=1.5, shapeName=strap4_b_capsule, center=<0 0.532 -1.2981>, rotation=<rot 90 0 0>}]
 strap4_d [{shapeType=Capsule, length=2.5, radius=1.25, shapeName=strap4_d_capsule, center=<-0.6369 0 0>, rotation=<rot 90 0 0>}]
 zipper_base [{shapeType=Capsule, length=1, radius=1, shapeName=zipper_base_capsule, rotation=<rot 90 0 0>}]
 zipper_end [{shapeType=Capsule, length=1, radius=1, shapeName=zipper_end_capsule, rotation=<rot 90 0 0>}]
}
constraints [
 {bone1Name=strap1_d, bone2Name=pack, swing1LimitDegrees=5, swing2LimitDegrees=7, twistLimitDegrees=5}
 {bone1Name=strap2_d, bone2Name=strap1_d, swing1LimitDegrees=5, swing2LimitDegrees=7, twistLimitDegrees=5}
 {bone1Name=strap3_d, bone2Name=strap2_d, swing1LimitDegrees=5, swing2LimitDegrees=7, twistLimitDegrees=5}
 {bone1Name=strap4_d, bone2Name=strap3_d, swing1LimitDegrees=5, swing2LimitDegrees=7, twistLimitDegrees=5}
 {bone1Name=strap1_b, bone2Name=pack, swing1LimitDegrees=2, swing2LimitDegrees=2, twistLimitDegrees=5}
 {bone1Name=strap2_b, bone2Name=strap1_b, swing1LimitDegrees=2, swing2LimitDegrees=2, twistLimitDegrees=5}
 {bone1Name=strap3_b, bone2Name=strap2_b, swing1LimitDegrees=2, swing2LimitDegrees=2, twistLimitDegrees=5}
 {bone1Name=strap4_b, bone2Name=strap3_b, swing1LimitDegrees=2, swing2LimitDegrees=2, twistLimitDegrees=5}
 {bone1Name=strap1_a, bone2Name=pack, swing1LimitDegrees=3, swing2LimitDegrees=2, twistLimitDegrees=5}
 {bone1Name=strap2_a, bone2Name=strap1_a, swing1LimitDegrees=3, swing2LimitDegrees=2, twistLimitDegrees=5}
 {bone1Name=strap3_a, bone2Name=strap2_a, swing1LimitDegrees=3, swing2LimitDegrees=2, twistLimitDegrees=5}
 {bone1Name=strap4_a, bone2Name=strap3_a, swing1LimitDegrees=3, swing2LimitDegrees=2, twistLimitDegrees=5}
 {bone1Name=gun4_base, bone2Name=pack, swing1LimitDegrees=5, swing2LimitDegrees=5, twistLimitDegrees=5}
 {
  bone1Name=gun4_end
  bone2Name=gun4_base
  swing1LimitDegrees=45
  swing1Motion=Free
  swing2LimitDegrees=45
  swing2Motion=Free
  twistLimitDegrees=45
  twistMotion=Free
 }
 {bone1Name=gun3_base, bone2Name=pack, swing1LimitDegrees=5, swing2LimitDegrees=5, twistLimitDegrees=5}
 {
  bone1Name=gun3_end
  bone2Name=gun3_base
  swing1LimitDegrees=45
  swing1Motion=Free
  swing2LimitDegrees=45
  swing2Motion=Free
  twistLimitDegrees=45
  twistMotion=Free
 }
 {bone1Name=gun2_base, bone2Name=pack, swing1LimitDegrees=5, swing2LimitDegrees=5, twistLimitDegrees=5}
 {
  bone1Name=gun2_end
  bone2Name=gun2_base
  swing1LimitDegrees=45
  swing1Motion=Free
  swing2LimitDegrees=45
  swing2Motion=Free
  twistLimitDegrees=45
  twistMotion=Free
 }
 {bone1Name=gun1_base, bone2Name=pack, swing1LimitDegrees=5, swing2LimitDegrees=5, twistLimitDegrees=5}
 {
  bone1Name=gun1_end
  bone2Name=gun1_base
  swing1LimitDegrees=45
  swing1Motion=Free
  swing2LimitDegrees=45
  swing2Motion=Free
  twistLimitDegrees=45
  twistMotion=Free
 }
 {bone1Name=zipper_base, bone2Name=pack, swing1LimitDegrees=5, swing2LimitDegrees=5, twistLimitDegrees=5}
 {bone1Name=zipper_end, bone2Name=zipper_base, swing1LimitDegrees=5, swing2LimitDegrees=5, twistLimitDegrees=5}
]
