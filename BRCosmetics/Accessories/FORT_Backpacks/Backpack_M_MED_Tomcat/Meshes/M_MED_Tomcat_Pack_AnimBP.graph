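# Animation Blueprint for the Tomcat backpack cosmetic that adds physics motion to the pack and its straps.
# From the reference pose, a RigidBody node runs the pack physics asset, then chained Anim Dynamics nodes add tightly limited sway to four strap bones.
# The pack bone is then nudged additively, and strap4_d is constrained equally between strap3_d and strap4_c.
# Copy Pose From Mesh is unconnected and the event graph has no logic.
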AnimBlueprint /BRCosmetics/Accessories/FORT_Backpacks/Backpack_M_MED_Tomcat/Meshes/M_MED_Tomcat_Pack_AnimBP
target_skeleton=&/M_MED_Tomcat_Pack_Skeleton
graphs [
 AnimGraph {nodes [
   Copy Pose From Mesh() {bCopyCurves=true, bUseAttachedParent=true, rootBoneToCopy=root}
   n1 = RigidBody(ComponentPose=Local To Component(LocalPose=Local Space Ref Pose())) {
    bOverrideWorldGravity=true
    overridePhysicsAsset=&/M_MED_Tomcat_Pack_Physics
    simulationSpace=BaseBoneSpace
    alphaBoolBlend {blendOption=Linear}
    baseBoneRef {boneName=root}
    componentAppliedLinearAccClamp=<1000 1000 1000>
    componentLinearAccScale=<0.2 0.2 0.25>
    componentLinearVelScale=<0.2 0.2 0.25>
    overrideWorldGravity {z=-980}
   }
   n2 = Anim Dynamics - Bone: strap1_c(ComponentPose=@n1, Alpha=1, ExternalForce="0,0,0") {
    bUseGravityOverride=true
    simulationSpace=RootRelative
    alphaBoolBlend {blendOption=Linear}
    boundBone {boneName=strap1_c}
    componentAppliedLinearAccClamp=<2000 2000 2000>
    componentLinearAccScale=<0.2 0.2 0.25>
    componentLinearVelScale=<0.2 0.2 0.25>
    gravityOverride {z=-980}
    physicsBodyDefinitions [{
      boundBone {boneName=strap1_c}
      boxExtents=<20 20 20>
      constraintSetup {angularLimitsMax=<2 3 0>, angularLimitsMin=<-2 -3 -2>, angularTarget=<0 0 0>}
      localJointOffset=<5 0 0>
     }]
   }
   n3 = Anim Dynamics - Bone: strap2_c(ComponentPose=@n2, Alpha=1, ExternalForce="0,0,0") {
    bUseGravityOverride=true
    simulationSpace=RootRelative
    alphaBoolBlend {blendOption=Linear}
    boundBone {boneName=strap2_c}
    componentAppliedLinearAccClamp=<2000 2000 2000>
    componentLinearAccScale=<0.2 0.2 0.25>
    componentLinearVelScale=<0.2 0.2 0.25>
    gravityOverride {z=-980}
    physicsBodyDefinitions [{
      boundBone {boneName=strap2_c}
      boxExtents=<20 20 20>
      constraintSetup {angularLimitsMax=<2 3 0>, angularLimitsMin=<-2 -3 -2>, angularTarget=<0 0 0>}
      localJointOffset=<5 0 0>
     }]
   }
   n4 = Anim Dynamics - Bone: strap3_c(ComponentPose=@n3, Alpha=1, ExternalForce="0,0,0") {
    bUseGravityOverride=true
    simulationSpace=RootRelative
    alphaBoolBlend {blendOption=Linear}
    boundBone {boneName=strap3_c}
    componentAppliedLinearAccClamp=<2000 2000 2000>
    componentLinearAccScale=<0.2 0.2 0.25>
    componentLinearVelScale=<0.2 0.2 0.25>
    gravityOverride {z=-980}
    physicsBodyDefinitions [{
      boundBone {boneName=strap3_c}
      boxExtents=<20 20 20>
      constraintSetup {angularLimitsMax=<2 3 0>, angularLimitsMin=<-2 -3 -2>, angularTarget=<0 0 0>}
      localJointOffset=<5 0 0>
     }]
   }
   n5 = Anim Dynamics - Bone: strap4_c(ComponentPose=@n4, Alpha=1, ExternalForce="0,0,0") {
    bUseGravityOverride=true
    simulationSpace=RootRelative
    alphaBoolBlend {blendOption=Linear}
    boundBone {boneName=strap4_c}
    componentAppliedLinearAccClamp=<2000 2000 2000>
    componentLinearAccScale=<0.2 0.2 0.25>
    componentLinearVelScale=<0.2 0.2 0.25>
    gravityOverride {z=-980}
    physicsBodyDefinitions [{
      boundBone {boneName=strap4_c}
      boxExtents=<20 20 20>
      constraintSetup {angularLimitsMax=<2 3 0>, angularLimitsMin=<-2 -3 -2>, angularTarget=<0 0 0>}
      localJointOffset=<5 0 0>
     }]
   }
   n6 = "Transform (Modify) Bone - Bone: pack"(ComponentPose=@n5, Translation="0.000,2.081,0.000", Rotation="0,0,0") {translationMode=BMM_Additive, alphaBoolBlend {blendOption=Linear}, boneToModify {boneName=pack}, translation {y=2.081}}
   n7 = "Constraint - strap4_d (PP)"!(ComponentPose=@n6, Alpha=1, ConstraintWeights_0=0.5, ConstraintWeights_1:float=0.5) -> (Pose) {
    alphaBoolBlend {blendOption=Linear}
    boneToModify {boneName=strap4_d}
    constraintSetup [{targetBone {boneName=strap3_d}}, {targetBone {boneName=strap4_c}}]
    constraintWeights [1,1]
   }
   Output Pose(Result=Component To Local(ComponentPose=@n7)) {layerGroup=None}
  ]}
 EventGraph {nodes [
   TryGetPawnOwner()
   Event BlueprintUpdateAnimation()
  ]}
]
variables []
